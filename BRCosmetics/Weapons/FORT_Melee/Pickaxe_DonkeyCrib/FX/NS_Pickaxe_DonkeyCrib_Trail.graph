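# Niagara swing trail for the DonkeyCrib pickaxe, built from two short-lived anim-trail ribbon emitters, TrailMain and TrailSecondary.
# Both spawn through the FN anim-trail modules, then fade color and alpha and taper ribbon width over life.
# Fade strength is scaled by time of day, and RGB is tinted by a system TOD color.
# TrailMain has a second ribbon renderer and computes a fading black, half-width DarkLines color and width.
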NiagaraSystem /BRCosmetics/Weapons/FORT_Melee/Pickaxe_DonkeyCrib/FX/NS_Pickaxe_DonkeyCrib_Trail
dependencies {
 usedDataInterfaces [&NiagaraDataInterfaceCurve, &NiagaraDataInterfaceSocketReader, &NiagaraDataInterfaceColorCurve]
 usedDynamicInputs [&ColorFromCurve]
 usedModules [
  &self:SystemSpawnScript.SystemScriptSource.SystemScriptGraph.NiagaraNodeAssignment_0.SetVariables_#1
  &self:SystemSpawnScript.SystemScriptSource.SystemScriptGraph.NiagaraNodeAssignment_2.SetVariables_#2
  &SystemState
 ]
}
emitters {
 TrailMain {
  emitterUpdateScript [
   &EmitterState()
   &FN_SpawnAnimTrail()
  ]
  particleSpawnScript [
   &FN_InitializeAnimTrailBoneIndecies() {module=FN_InitializeAnimTrailBoneIndecies001}
   &FN_InitializeAnimTrail(
    AnimTrailWidth=1.4
    Color=&MultiplyLinearColor(Linear Color=<rgba 0.95 0.9583 1 3>, Scale Factor=<1 1 1 0.3>)
    Lifetime=0.16
   ) {module=FN_InitializeAnimTrail001}
   &SystemLocation(MaskSpawnGroup=false)
  ]
  particleUpdateScript [
   &ScaleColor(
    Scale Alpha=&FloatFromCurve(
     FloatCurve=DataInterface() {propertyValues {Curve {keys [
         {interpMode=RCIM_Cubic, leaveTangent=-2.7775, tangentMode=RCTM_Break, time=0, value=1}
         {arriveTangent=-0.0411, interpMode=RCIM_Cubic, tangentMode=RCTM_Break, time=1, value=0}
        ]}}}
     Scale Curve=&Multiply_Float(A=0.5, B=&FNTimeOfDayScalar(Night Scale=0.8, Morning Scale=0.9, Evening Scale=0.9))
    )
    Scale RGB=&MakeVectorFromLinearColorRGB(Color=&Multiply_LinearColorByFloat(
      Linear Color=$System.TOD
      Float=&FloatFromCurve(
       FloatCurve=DataInterface() {propertyValues {Curve {keys [
           {interpMode=RCIM_Cubic, leaveTangent=-3.9738, tangentMode=RCTM_Break, time=0, value=1}
           {arriveTangent=-0.0411, interpMode=RCIM_Cubic, leaveTangent=0.2878, tangentMode=RCTM_Break, time=0.8, value=0}
          ]}}}
       Scale Curve=&Multiply_Float(A=1, B=&FNTimeOfDayScalar(Night Scale=0.8, Morning Scale=0.9, Evening Scale=0.9))
      )
     ))
    ScaleA=true
    ScaleRGB=true
   ) {module=ScaleColor001}
   &FN_UpdateAnimTrail()
   &ScaleRibbonWidth(Ribbon Width Scale=&FloatFromCurve(FloatCurve=DataInterface() {propertyValues {Curve {keys [
         {arriveTangent=-1.5432, interpMode=RCIM_Cubic, leaveTangent=-1.5432, tangentMode=RCTM_User, time=0, value=1}
         {arriveTangent=-0.2684, interpMode=RCIM_Cubic, leaveTangent=-0.2684, tangentMode=RCTM_User, time=1, value=0.2}
        ]}}}))
   /SetVariables(
    Particles.DarkLinesColor:LinearColor=&MakeLinearColorFromVectorAndFloat(
     "Vector (RGB)"=<0 0 0>
     "Float (Alpha)"=&FloatFromCurve(
      FloatCurve=DataInterface() {propertyValues {Curve {keys [
          {interpMode=RCIM_Cubic, leaveTangent=-2.2093, tangentMode=RCTM_Break, time=0, value=1}
          {arriveTangent=-0.0411, interpMode=RCIM_Cubic, tangentMode=RCTM_Break, time=1, value=0}
         ]}}}
      Scale Curve=&Multiply_Float(A=0.8, B=&FNTimeOfDayScalar(Night Scale=0.8, Morning Scale=0.9, Evening Scale=0.9))
     )
    )
    Particles.DarkLinesWidth:NiagaraFloat=&Multiply_Float(A=$Particles.RibbonWidth, B=0.5)
   ) {isSetParameters=true}
  ]
  rendererClasses [&NiagaraRibbonRendererProperties]
  renderers [
   {rendererClass=&NiagaraRibbonRendererProperties}
   {rendererClass=&NiagaraRibbonRendererProperties, rendererIndex=1}
  ]
 }
 TrailSecondary {
  emitterUpdateScript [
   &EmitterState()
   &FN_SpawnAnimTrail()
  ]
  particleSpawnScript [
   &FN_InitializeAnimTrailBoneIndecies() {module=FN_InitializeAnimTrailBoneIndecies001}
   &FN_InitializeAnimTrail(
    AnimTrailWidth=1.4
    Color=&MultiplyLinearColor(Linear Color=<rgba 0.665 0.6708 0.7 3>, Scale Factor=<1 1 1 0.7>)
    Lifetime=0.22
   ) {module=FN_InitializeAnimTrail001}
   &SystemLocation(MaskSpawnGroup=false)
  ]
  particleUpdateScript [
   &ScaleColor(
    Scale Alpha=&FloatFromCurve(
     FloatCurve=DataInterface() {propertyValues {Curve {keys [
         {interpMode=RCIM_Cubic, leaveTangent=-2.6214, tangentMode=RCTM_Break, time=0, value=1}
         {arriveTangent=-0.0411, interpMode=RCIM_Cubic, leaveTangent=0.2878, tangentMode=RCTM_Break, time=1, value=0}
        ]}}}
     Scale Curve=&Multiply_Float(A=0.5, B=&FNTimeOfDayScalar(Night Scale=0.8, Morning Scale=0.9, Evening Scale=0.9))
    )
    Scale RGB=&MakeVectorFromLinearColorRGB(Color=&Multiply_LinearColorByFloat(
      Linear Color=$System.TOD
      Float=&FloatFromCurve(
       FloatCurve=DataInterface() {propertyValues {Curve {keys [
           {interpMode=RCIM_Cubic, leaveTangent=-3.9738, tangentMode=RCTM_Break, time=0, value=1}
           {arriveTangent=-0.0411, interpMode=RCIM_Cubic, leaveTangent=0.2878, tangentMode=RCTM_Break, time=0.8, value=0}
          ]}}}
       Scale Curve=&Multiply_Float(A=1, B=&FNTimeOfDayScalar(Night Scale=0.8, Morning Scale=0.9, Evening Scale=0.9))
      )
     ))
    ScaleA=true
    ScaleRGB=true
   ) {module=ScaleColor001}
   &FN_UpdateAnimTrail()
   &ScaleRibbonWidth(Ribbon Width Scale=&FloatFromCurve(FloatCurve=DataInterface() {propertyValues {Curve {keys [
         {arriveTangent=-1.5432, interpMode=RCIM_Cubic, leaveTangent=-1.5432, tangentMode=RCTM_User, time=0, value=1}
         {arriveTangent=-0.2684, interpMode=RCIM_Cubic, leaveTangent=-0.2684, tangentMode=RCTM_User, time=1, value=0.2}
        ]}}}))
  ]
  rendererClasses [&NiagaraRibbonRendererProperties]
  renderers [{rendererClass=&NiagaraRibbonRendererProperties}]
 }
}
summary {
 systemName=NS_Pickaxe_DonkeyCrib_Trail
 emitters [
  {emitterName=TrailSecondary, rendererClasses [&NiagaraRibbonRendererProperties]}
  {emitterName=TrailMain, rendererClasses [&NiagaraRibbonRendererProperties]}
 ]
 userVariables [
  User.bIsFrontEnd {type=NiagaraBool}
  User.bIsFrontEndPreview {type=NiagaraBool}
  User.RibbonIDOffset {type=NiagaraInt32}
 ]
}
