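# Animation Blueprint for the DonkeyCrib pickaxe mesh, switching between idle and swinging states.
# Idle plays a non-looping idle sequence and moves to Swinging when bIsPlayingMeleeAnim, which is not declared here, is true.
# Swinging outputs only the local ref pose, yet its return transition uses the automatic sequence-player rule.
# The state machine feeds DefaultSlot, letting montages override it; the event graph holds only unconnected nodes.
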AnimBlueprint /BRCosmetics/Weapons/FORT_Melee/Pickaxe_DonkeyCrib/Meshes/DonkeyCrib_Axe_AnimBP
target_skeleton=&/DonkeyCrib_Axe_Skeleton
graphs [
 AnimGraph {nodes [
   Local Space Ref Pose()
   AnimGraphNode_StateMachine_0 = Idle_State() {is=title AnimGraphNode_StateMachine}
   Output Pose(Result=Slot 'DefaultSlot'(Source=@AnimGraphNode_StateMachine_0)) {layerGroup=None}
   Sequence Player 'DonkeyCrib_Idle_Axe'() {bLoopAnimation=false, sequence=/BRCosmetics/Weapons/FORT_Melee/Pickaxe_DonkeyCrib/Anim/DonkeyCrib_Idle_Axe}
   AnimGraphNode_SubInput_0 = Input Pose()
  ]}
 Idle_State {
  entry=Idle
  path=AnimGraph.AnimGraphNode_StateMachine_0.Idle_State
  ids=id
  nodes [
   AnimStateNode_0 = Idle()
   AnimStateNode_1 = Swinging() {is=title AnimStateNode}
   AnimStateTransitionNode_0 = AnimStateTransitionNode()
   AnimStateTransitionNode_2 = AnimStateTransitionNode() {bAutomaticRuleBasedOnSequencePlayerInState=true}
  ]
  links [
   {
    dst=Swinging
    rule=AnimGraph.AnimGraphNode_StateMachine_0.Idle_State.AnimStateTransitionNode_0.Transition
    src=Idle
    via=AnimStateTransitionNode_0
   }
   {
    dst=Idle
    rule=AnimGraph.AnimGraphNode_StateMachine_0.Idle_State.AnimStateTransitionNode_2.Transition
    src=Swinging
    via=AnimStateTransitionNode_2
   }
  ]
 }
 Idle {
  path=AnimGraph.AnimGraphNode_StateMachine_0.Idle_State.AnimStateNode_0.Idle
  nodes [Output Animation Pose(Result=Sequence Player 'DonkeyCrib_Idle_Axe'() {bLoopAnimation=false, sequence=&DonkeyCrib_Idle_Axe})]
 }
 Swinging {
  path=AnimGraph.AnimGraphNode_StateMachine_0.Idle_State.AnimStateNode_1.Swinging
  nodes [Output Animation Pose(Result=Local Space Ref Pose())]
 }
 Transition {
  path=AnimGraph.AnimGraphNode_StateMachine_0.Idle_State.AnimStateTransitionNode_0.Transition
  nodes [Result(bCanEnterTransition=Get bIsPlayingMeleeAnim())]
 }
 Transition {path=AnimGraph.AnimGraphNode_StateMachine_0.Idle_State.AnimStateTransitionNode_2.Transition, nodes [Result()]}
 EventGraph {nodes [
   TryGetPawnOwner()
   Event BlueprintUpdateAnimation()
  ]}
]
variables []
